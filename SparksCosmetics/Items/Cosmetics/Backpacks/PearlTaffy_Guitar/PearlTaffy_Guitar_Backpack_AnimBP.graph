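# Animation Blueprint for the PearlTaffy guitar backpack cosmetic that shifts and tilts the guitar as the spine bends.
# Each update reads the yaw of spine_05's delta from ref pose relative to spine_01 on the owner skeletal mesh component.
# Clamped range maps turn negative yaw into rotation and translation, summing consecutive ranges for piecewise curves on roll and translation.
# The AnimGraph applies both additively in bone space to the offset bone.
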AnimBlueprint /SparksCosmetics/Items/Cosmetics/Backpacks/PearlTaffy_Guitar/PearlTaffy_Guitar_Backpack_AnimBP
target_skeleton=/SparksCharacterCommon/Instrument/Guitar/Generic/Meshes/Generic_Guitar_Skeleton
graphs [
 AnimGraph {nodes [
   n1 = "Transform (Modify) Bone - Bone: offset"!(
    ComponentPose=Local To Component(LocalPose=Input Pose())
    bAlphaBoolEnabled:bool=True
    Translation=Get backpack_translation()
    Rotation=Get backpack_rotate()
    Scale="1,1,1"
   ) -> (Pose) {
    alphaInputType=Bool
    rotationMode=BMM_Additive
    rotationSpace=BCS_BoneSpace
    scaleMode=BMM_Additive
    scaleSpace=BCS_BoneSpace
    translationMode=BMM_Additive
    translationSpace=BCS_BoneSpace
    alphaBoolBlend {blendOption=Linear}
    boneToModify {boneName=offset}
   }
   Output Pose(Result=Component To Local(ComponentPose=@n1))
  ]}
 EventGraph {nodes [
   Event BlueprintUpdateAnimation()
   n1 = GetOwnerSkeletalMeshComponent()
   n2 = BreakRotator(InRot=BreakTransform(InTransform=GetDeltaTransformFromRefPose(self=@n1, BoneName=spine_05, BaseName=spine_01)).Rotation)
   n3 = MapRangeClamped(Value=@n2.Yaw, InRangeA=-30, InRangeB=-80, OutRangeB=55)
   n4 = MapRangeClamped(Value=@n2.Yaw, InRangeA=-8, InRangeB=-30, OutRangeB=10)
   > Set backpack_rotate(backpack_rotate=MakeRotator(
     Roll=float + float(A=@n4, B=@n3)
     Pitch=MapRangeClamped(Value=@n2.Yaw, InRangeA=-30, InRangeB=-80, OutRangeB=45)
     Yaw=MapRangeClamped(Value=@n2.Yaw, InRangeA=-30, InRangeB=-80, OutRangeB=-65)
    ))
   n5 = MapRangeClamped(Value=@n2.Yaw, InRangeA=-30, InRangeB=-50, OutRangeB=-10)
   n6 = MapRangeClamped(Value=@n2.Yaw, InRangeA=0, InRangeB=-30, OutRangeB=-4)
   n7 = MapRangeClamped(Value=@n2.Yaw, InRangeA=-30, InRangeB=-50, OutRangeB=4)
   n8 = MapRangeClamped(Value=@n2.Yaw, InRangeA=0, InRangeB=-30, OutRangeB=2)
   n9 = MapRangeClamped(Value=@n2.Yaw, InRangeA=0, InRangeB=-20, OutRangeB=5)
   n10 = MapRangeClamped(Value=@n2.Yaw, InRangeA=-20, InRangeB=-50, OutRangeB=10)
   > Set backpack_translation(backpack_translation=MakeVector(X=float + float(A=@n9, B=@n10), Y=float + float(A=@n6, B=@n5), Z=float + float(A=@n8, B=@n7)))
   TryGetPawnOwner()
  ]}
]
variables [
 backpack_rotate {type=Rotator}
 backpack_translation {type=Vector}
]
